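; Windmill impulse actor that launches player and AI pawns overlapping its collision box, sized from BoundsSize at construction.
; Distances to points cached ahead of and behind the box choose a launch along forward or backward, toward the side opposite the pawn.
; A capsule ceiling trace selects bounce height before LaunchCharacter, followed by a launched gameplay effect, fall-damage immunity and force feedback.
; Other overlapping actors trigger player ejection from vehicles in the box.
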(event EventBeginPlay
  (Collision|BindEventtoOnComponentBeginOverlap (Variables|Default|GetCollisionBox) (delegate Custom|Touched))
  (Variables|Default|SetWaterfallForwardVector (+ (Transformation|GetWorldLocation (Variables|Default|GetCollisionBox)) (* (Math|Vector|Normalize (Transformation|GetForwardVector (Variables|Default|GetCollisionBox))) "512.000000,512.000000,512.000000")))
  (Variables|Default|SetWaterfallBehindVector (+ (Transformation|GetWorldLocation (Variables|Default|GetCollisionBox)) (* (Math|Vector|Normalize (Math|Vector|NegateVector (Transformation|GetForwardVector (Variables|Default|GetCollisionBox)))) "512.000000,512.000000,512.000000"))))

(event Custom|Touched
  (params (OverlappedComponent :type Primitive%20Component%20Object%20Reference) (OtherActor :type Actor%20Object%20Reference) (OtherComp :type Primitive%20Component%20Object%20Reference) (OtherBodyIndex :type Integer) (bFromSweep :type Boolean) (SweepResult :type Hit%20Result%20Structure%20%28by%20ref%29))
  (bind _scene (Variables|Default|GetScene))
  (bind _returnvalue (Transformation|GetForwardVector _scene))
  (bind _scene_2 (Variables|Default|GetScene))
  (bind _returnvalue_2 (Transformation|GetForwardVector _scene_2))
  (bind _returnvalue_3 (Math|Vector|NegateVector _returnvalue))
  (bind _overlappawn (Variables|Default|GetOverlapPawn))
  (bind _overlapcontroller (Variables|Default|GetOverlapController))
  (bind _bouncexyvelocitymofifier (Variables|Default|GetBounceXYVelocityMofifier))
  (bind _bouncexyvelocitymofifier_2 (Variables|Default|GetBounceXYVelocityMofifier))
  (bind _overlappawn_2 (Variables|Default|GetOverlapPawn))
  (Variables|Default|SetBounceLocation (Transformation|GetActorLocation OtherActor))
  (bind _asfort_player_pawn (Utilities|Casting|CastToFortPlayerPawn OtherActor)
    (:then
      (Variables|Default|SetOverlapPawn _asfort_player_pawn))
    (:CastFailed
      (bind _asfort_aipawn (Utilities|Casting|CastToFortAIPawn OtherActor)
        (:then
          (Variables|Default|SetOverlapPawn _asfort_aipawn))
        (:CastFailed
          ;[1 1 1] Eject players from vehicles
          (for _array_element (Collision|GetOverlappingActors (Variables|Default|GetCollisionBox) "/Script/FortniteGame.FortAthenaVehicle" :!class /Script/Engine.PrimitiveComponent)
            (Vehicle|EjectAllPlayers(Message) _array_element))))))
  (bind _isforward (CallFunction|CheckWaterFallBounceDirection))
  (Variables|WaterFall|SetIsForwardDirection _isforward)
  (Utilities|IsValid _overlappawn_2
    (:Is%20Valid
      (if (or (GameplayTags|HasMatchingGameplayTag (Ability|GetAbilitySystemComponent _overlappawn_2) "(TagName=\"Gameplay.Action.Player.Launched\")") (Variables|/Script/FortniteGame.FortPawn|GetbIsDBNO _overlappawn_2))
        _
        (else
          _)))
    (:Is%20Not%20Valid
      _))
  (bind _asfort_player_controller (Utilities|Casting|CastToFortPlayerController (Pawn|GetController _overlappawn_2))
    (:then
      ;[1 1 1] Ceiling check
      (Variables|Default|SetOverlapController _asfort_player_controller))
    (:CastFailed
      _))
  (bind _success (CallFunction|CeilingCheck))
  (if _success
    (Variables|Default|SetBounceHeight 1.1e+03)
    (Variables|Default|SetJumpHeight 1e+02)
    (else
      (Variables|Default|SetBounceHeight (Variables|Default|GetNoCeilingBounceHeight))
      (Variables|Default|SetJumpHeight 5e+02)))
  (Variables|Default|SetFinalBounceJumpHeight (Variables|Default|GetBounceHeight))
  (bind _asfort_player_pawn_2 (Utilities|Casting|CastToFortPlayerPawn (Variables|Default|GetOverlapPawn))
    (:then
      (RemoteControlPawn|ForceDetachfromRemoteControlledPawn _asfort_player_pawn_2))
    (:CastFailed
      _))
  (if (Variables|WaterFall|GetIsForwardDirection)
    (Variables|Default|SetFinalXVector (* (.x _returnvalue_2) _bouncexyvelocitymofifier_2))
    (Variables|Default|SetFinalYVector (* (.y _returnvalue_2) _bouncexyvelocitymofifier_2))
    (else
      (Variables|Default|SetFinalXVector (* (.x _returnvalue_3) _bouncexyvelocitymofifier))
      (Variables|Default|SetFinalYVector (* (.y _returnvalue_3) _bouncexyvelocitymofifier))))
  (Character|LaunchCharacter (Variables|Default|GetOverlapPawn) (Math|Vector|MakeVector (Variables|Default|GetFinalXVector) (Variables|Default|GetFinalYVector) (Variables|Default|GetFinalBounceJumpHeight)) true true)
  (GameplayEffects|ApplyGameplayEffectToSelf :self (Ability|GetAbilitySystemComponent _overlappawn) :GameplayEffectClass (Variables|Default|GetGE_PlayerLaunched) :Level 0.0)
  (Variables|/Script/FortniteGame.FortPawn|SetbIgnoreNextFallingDamage true _overlappawn)
  (Utilities|IsValid _overlapcontroller
    (:Is%20Valid
      (Game|Feedback|LocalPlayForceFeedback _overlapcontroller (Variables|Default|GetLaunchForceFeedback) false "TireLaunch"))
    (:Is%20Not%20Valid
      _))
  (Variables|Default|SetOverlapPawn 0)
  (Variables|Default|SetOverlapController 0))

(fn CeilingCheck
  (bind _overlappawn (Variables|Default|GetOverlapPawn)
    (:then
      (Utilities|Array|Add (LocalVariables|GetActorsToIgnore) _overlappawn)
      (Utilities|Array|AppendArray (LocalVariables|GetActorsToIgnore) (Actor|GetAttachedActors _overlappawn true false))
      (for _array_element (Actor|GetAttachedActors _overlappawn true false)
        (Utilities|Array|AppendArray (LocalVariables|GetActorsToIgnore) (Actor|GetAttachedActors _array_element true false))))
    (:else
      _))
  (bind (_outhit _returnvalue) (Collision|CapsuleTraceByChannel (Math|Vector|MakeVector (.x (Transformation|GetWorldLocation (Variables|Transformation|GetRootComponent))) (.y (Transformation|GetWorldLocation (Variables|Transformation|GetRootComponent))) (+ (.z (Transformation|GetWorldLocation (Variables|Transformation|GetRootComponent))) 200)) (Math|Vector|MakeVector (.x (Transformation|GetWorldLocation (Variables|Transformation|GetRootComponent))) (.y (Transformation|GetWorldLocation (Variables|Transformation|GetRootComponent))) (+ (.z (Transformation|GetWorldLocation (Variables|Transformation|GetRootComponent))) 340)) 50 100 "TraceTypeQuery9" false (LocalVariables|GetActorsToIgnore)))
  (return _returnvalue))

(fn CheckJumpingOrientation
  (bind (_returnvalue_x _returnvalue_y) (Math|Vector|BreakVector (Transformation|GetActorForwardVector)))
  (bind _orientationmultiplier (Variables|Default|GetOrientationMultiplier))
  (if (== (Math|Float|Round (Math|Float|Absolute(Float) _returnvalue_x)) 1)
    (return (Math|Vector|vector*float (Math|Vector|MakeVector _returnvalue_x _returnvalue_y) _orientationmultiplier))
    (else
      (return (Math|Vector|vector*float (Math|Vector|MakeVector _returnvalue_x _returnvalue_y) (* _orientationmultiplier -1.0))))))

(fn CheckWaterFallBounceDirection
  (bind _distance (Class|MissionBlueprintFunctionLibrary|DistanceBetweenTwoVectors (Transformation|GetActorLocation (Variables|Default|GetOverlapPawn)) (Variables|Default|GetWaterfallForwardVector)))
  (LocalVariables|SetForwardDistanceLocal _distance)
  (bind _distance_2 (Class|MissionBlueprintFunctionLibrary|DistanceBetweenTwoVectors (Transformation|GetActorLocation (Variables|Default|GetOverlapPawn)) (Variables|Default|GetWaterfallBehindVector)))
  (LocalVariables|SetBehindDistanceLocal _distance_2)
  (Utilities|Array|Add (LocalVariables|GetDistancesLocal) (LocalVariables|GetForwardDistanceLocal))
  (Utilities|Array|Add (LocalVariables|GetDistancesLocal) (LocalVariables|GetBehindDistanceLocal))
  (return (>= (Utilities|Array|Get(acopy) (LocalVariables|GetDistancesLocal)) (Utilities|Array|Get(acopy) (LocalVariables|GetDistancesLocal) 1))))

(fn ConstructionScript
  (Components|Box|SetBoxExtent (Variables|Default|GetCollisionBox) (Variables|WaterFall|GetBoundsSize) true))
